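# Sound cue for aiming the Spicy Soda item, played as close-range player foley.
# A mixer layers one of four random Shield Jar Large jostle sounds, with slight pitch variation, over a quiet, pitched-down Spicy Soda drop sound.
# It uses close prop attenuation and overrides concurrency with a rule that prevents new instances once the limit is reached.
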SoundCue /FNGameplayCues/Athena/Sounds/Items/SpicySoda/SpicySoda_Aim_Cue
root=Mixer_0
nodes [
 {id=Mixer_0, children [Modulator_1,Modulator_0], properties {inputVolume [1,1]}}
 {id=Modulator_0, children [WavePlayer_0], properties {pitchMax=0.8, pitchMin=0.75, volumeMax=0.3, volumeMin=0.3}}
 {id=Modulator_1, children [Random_1], properties {pitchMax=1.1}}
 {id=Random_1, children [WavePlayer_4,WavePlayer_5,WavePlayer_6,WavePlayer_7], properties {weights [1,1,1,1]}}
 {id=WavePlayer_0, properties {soundWaveAssetPtr=/Game/Athena/Sounds/Items/SpicySoda/SpicySoda_Drop}}
 {id=WavePlayer_4, properties {soundWaveAssetPtr=/Game/Athena/Sounds/Items/Shield_Jar_Large_Jostle_01}}
 {id=WavePlayer_5, properties {soundWaveAssetPtr=/Game/Athena/Sounds/Items/Shield_Jar_Large_Jostle_02}}
 {id=WavePlayer_6, properties {soundWaveAssetPtr=/Game/Athena/Sounds/Items/Shield_Jar_Large_Jostle_03}}
 {id=WavePlayer_7, properties {soundWaveAssetPtr=/Game/Athena/Sounds/Items/Shield_Jar_Large_Jostle_04}}
]
properties {
 attenuationSettings=/Game/Sounds/Attenuation/Fort_Prop_Close_Att
 bOverrideConcurrency=true
 maxDistance=1800
 soundClassObject=&Fort_Player_Foley
 volumeMultiplier=1
 concurrencyOverrides {resolutionRule=PreventNew}
}
