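# Sound cue for the Spicy Soda item's explosion, routed to the large explosives sound class.
# A Random node picks one of two equally weighted explosion waves, feeding a smoke-grenade Attenuation node, then a Modulator that varies pitch while holding volume steady.
# The cue also overrides attenuation with its own falloff and shape, and overrides concurrency to prevent new instances once the limit is reached.
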SoundCue /FNGameplayCues/Athena/Sounds/Items/SpicySoda/SpicySoda_Explode_Cue
root=Modulator_12
nodes [
 {id=Attenuation_1, children [Random_0], properties {attenuationSettings=/Game/Sounds/Attenuation/Fort_Grenade_Smoke_Att}}
 {id=Modulator_12, children [Attenuation_1], properties {pitchMin=0.9, volumeMax=1, volumeMin=1}}
 {id=Random_0, children [WavePlayer_1,WavePlayer_2], properties {weights [1,1]}}
 {id=WavePlayer_1, properties {soundWaveAssetPtr=&/SpicySoda_Explosion_01}}
 {id=WavePlayer_2, properties {soundWaveAssetPtr=&/SpicySoda_Explosion_02}}
]
properties {
 attenuationSettings=&Fort_Grenade_Smoke_Att
 bOverrideAttenuation=true
 bOverrideConcurrency=true
 maxDistance=6800
 soundClassObject=/Game/Sounds/FortSoundClassesAndModes/Fort_Explosives_Large_PSM
 volumeMultiplier=0.2477
 attenuationOverrides {falloffDistance=5000, attenuationShapeExtents {x=600}}
 concurrencyOverrides {resolutionRule=PreventNew}
}
